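# LiveStreamLetEmKnow is an in-match callout banner announcing a live event, or picture-in-picture availability in Creative/Playground playlists.
# Update based on Mode picks text and delay by playlist; Creative shows it only when the picture-in-picture setting is not Show, other modes always.
# ShowAdvertisement waits that delay, scales the skewed banner in, holds it, then reverses the animation and collapses the widget.
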WidgetBlueprint /FortniteCoreUI/Frontend/LiveStreamVIdeo/LiveStreamLetEmKnow
animations [IntroOutro {
  binding_count=1
  length_seconds=0.25
  object=&self:IntroOutro
  sequence {bindings [RootOverlay {
     id=#1
     possessed_class=Overlay
     tracks [{
       type=MovieScene2DTransformTrack
       sections [{
         type=MovieScene2DTransformSection
         end_frame=5
         start_frame=0
         channels [
          Translation.X {type=MovieSceneScriptingFloatChannel}
          Translation.Y {type=MovieSceneScriptingFloatChannel}
          Angle {type=MovieSceneScriptingFloatChannel}
          Scale.X {type=MovieSceneScriptingFloatChannel, keys [[0,0],[5,1]]}
          Scale.Y {type=MovieSceneScriptingFloatChannel, keys [[0,0],[5,1]]}
          Shear.X {type=MovieSceneScriptingFloatChannel}
          Shear.Y {type=MovieSceneScriptingFloatChannel}
         ]
        }]
      }]
    }]}
 }]
graphs [
 EventGraph {nodes [
   Event ShowAdvertisement() {is=title K2Node_Event}
   > n1 = Update based on Mode!(execute, self) -> (then, ShouldShowMessage:bool, Initial Delay:double)
   > Branch(Condition=@n1)
   > Delay(Duration=@n1.Initial Delay)
   > SetVisibility(InVisibility=HitTestInvisible)
   > PlayAnimationForward(InAnimation=Get IntroOutro() -> (IntroOutro:Widget Animation Object Reference))
   > Delay(Duration=5)
   > PlayAnimationReverse(InAnimation=Get IntroOutro())
   > Delay(Duration=2)
   > SetVisibility(InVisibility=Collapsed)
  ]}
 Update based on Mode {nodes [
   Update based on Mode() {is=title K2Node_FunctionEntry}
   > n1 = Set isCreative(isCreative:bool=IsCreativeOrPlaygroundModePlaylist!() -> (ReturnValue:bool)) -> (Output_Get:bool)
   > "SetText (Text)"(
    self=Get Text_Callout() -> (Text_Callout:Common Text Object Reference)
    InText=Select(Option 0:Text=Get Papaya Text() -> (Papaya Text:Text), Option 1:Text=Get Creative Text() -> (Creative Text:Text), Index=@n1) -> (ReturnValue:Text)
   )
   > n2 = Branch(Condition=Get isCreative() -> (isCreative:bool))
   > n3 = GetGameUserSettings!(execute) -> (then, ReturnValue:Game User Settings Object Reference)
   > n4 = Cast To FortGameUserSettings!(execute, Object=@n3) -> (then, CastFailed, AsFort Game User Settings:Fort Game User Settings Object Reference) {is=title K2Node_DynamicCast}
   > Return Node!(
    execute
    ShouldShowMessage:bool="Not Equal (Enum)"!(
     A:EShowInGamePictureInPicture Enum=Get ShowLiveStreamPictureInPictureInMatchV2!(self:Fort Game User Settings Object Reference=@n4) -> (ShowLiveStreamPictureInPictureInMatchV2:EShowInGamePictureInPicture Enum=Default)
     B:EShowInGamePictureInPicture Enum=Show
    ) -> (ReturnValue:bool)
    Initial Delay:double=Get Creative Delay() -> (Creative Delay:double)
   )
   Return Node(execute=@n2.else, ShouldShowMessage=true, Initial Delay=Get Papaya Delay() -> (Papaya Delay:double))
  ]}
]
variables [
 Creative Delay {type=number}
 Creative Text {type=string, default=PICTURE IN PICTURE IS AVAILABLE}
 Papaya Delay {type=number, default=3}
 Papaya Text {type=string, default=A LIVE EVENT IS HAPPENING NOW}
]
widgets {
 info {inheritedWidgetCount=0, parentClass=/Script/FortniteUI.LiveStreamInMatchAd, rootWidgetClass=&Overlay, widgetCount=11}
 widgets [
  RootOverlay {widgetClassPath=&Overlay, widget=RootOverlay, properties {renderTransformPivot=<1.1 -0.1>}}
  CommonBorderShadow {
   widgetClassPath=&CommonBorder
   parent=RootOverlay
   widget=CommonBorderShadow
   properties {
    contentColorAndOpacity=<rgba 0.0052 0.0052 0.0056 0.4>
    padding=<ltrb 0 0 0 0>
    renderTransform {angle=0, scale=<1 1>, shear=<-2 0>, translation=<4 5>}
   }
   slot {class=OverlaySlot, properties {horizontalAlignment=HAlign_Fill, verticalAlignment=VAlign_Fill, padding=<ltrb 0 -5 0 0>}}
  }
  OverlayShadow {
   widgetClassPath=&Overlay
   parent=CommonBorderShadow
   widget=OverlayShadow
   slot {class=BorderSlot, properties {padding=<ltrb 0 0 0 0>}}
  }
  Image_0 {
   widgetClassPath=&Image
   parent=OverlayShadow
   widget=Image_0
   properties {
    brush {
     bIsDynamicallyLoaded=false
     drawAs=Image
     imageType=NoImage
     mirroring=NoMirror
     resourceName=None
     resourceObject=/Game/UI/Frontend/Showdown/Materials/M_UI_LiveViolatorTail_Inst
     tiling=NoTile
     imageSize=<32 32>
     margin=<ltrb 0 0 0 0>
     outlineSettings {
      bUseBrushTransparency=false
      roundingType=HalfHeightRadius
      width=0
      color {colorUseRule=UseColor_Specified, specifiedColor=<rgba 0 0 0 0>}
      cornerRadii=<0 0 0 0>
     }
     tintColor {colorUseRule=UseColor_Specified, specifiedColor=<rgba 1 1 1 1>}
     uVRegion {bIsValid=false, max=<0 0>, min=<0 0>}
    }
    colorAndOpacity=<rgba 1 0 0 1>
    renderTransform {angle=-155, scale=<1 1>, shear=<0 0>, translation=<20 -20>}
   }
   slot {class=OverlaySlot, properties {horizontalAlignment=HAlign_Right}}
  }
  Image_1 {
   widgetClassPath=&Image
   parent=OverlayShadow
   widget=Image_1
   properties {
    brush {
     bIsDynamicallyLoaded=false
     drawAs=Box
     imageType=NoImage
     mirroring=NoMirror
     resourceName=None
     resourceObject=/Game/UI/Frontend/Showdown/Materials/M_UI_SkewBox_LiveBG
     tiling=NoTile
     imageSize=<32 32>
     margin=<ltrb 0.03 0 0.03 0>
     outlineSettings {
      bUseBrushTransparency=false
      roundingType=HalfHeightRadius
      width=0
      color {colorUseRule=UseColor_Specified, specifiedColor=<rgba 0 0 0 0>}
      cornerRadii=<0 0 0 0>
     }
     tintColor {colorUseRule=UseColor_Specified, specifiedColor=<rgba 1 1 1 1>}
     uVRegion {bIsValid=false, max=<0 0>, min=<0 0>}
    }
    colorAndOpacity=<rgba 1 0 0 1>
    renderTransform {angle=0, scale=<1 1>, shear=<-14 0>, translation=<0 0>}
   }
   slot {class=OverlaySlot, properties {horizontalAlignment=HAlign_Fill, verticalAlignment=VAlign_Fill}}
  }
  Overlay_3 {
   widgetClassPath=&Overlay
   parent=RootOverlay
   widget=Overlay_3
   slot {class=OverlaySlot, properties {horizontalAlignment=HAlign_Fill, verticalAlignment=VAlign_Fill, padding=<ltrb 0 -5 0 0>}}
  }
  Image_Tail {
   widgetClassPath=&Image
   parent=Overlay_3
   widget=Image_Tail
   properties {
    brush {
     bIsDynamicallyLoaded=false
     drawAs=Image
     imageType=NoImage
     mirroring=NoMirror
     resourceName=None
     resourceObject=&M_UI_LiveViolatorTail_Inst
     tiling=NoTile
     imageSize=<32 32>
     margin=<ltrb 0 0 0 0>
     outlineSettings {
      bUseBrushTransparency=false
      roundingType=HalfHeightRadius
      width=0
      color {colorUseRule=UseColor_Specified, specifiedColor=<rgba 0 0 0 0>}
      cornerRadii=<0 0 0 0>
     }
     tintColor {colorUseRule=UseColor_Specified, specifiedColor=<rgba 1 1 1 1>}
     uVRegion {bIsValid=false, max=<0 0>, min=<0 0>}
    }
    colorAndOpacity=<rgba 1 0 0 1>
    renderTransform {angle=-155, scale=<1 1>, shear=<0 0>, translation=<20 -20>}
   }
   slot {class=OverlaySlot, properties {horizontalAlignment=HAlign_Right}}
  }
  Image_LiveBG {
   widgetClassPath=&Image
   parent=Overlay_3
   widget=Image_LiveBG
   properties {
    brush {
     bIsDynamicallyLoaded=false
     drawAs=Box
     imageType=NoImage
     mirroring=NoMirror
     resourceName=None
     resourceObject=&M_UI_SkewBox_LiveBG
     tiling=NoTile
     imageSize=<32 32>
     margin=<ltrb 0.03 0 0.03 0>
     outlineSettings {
      bUseBrushTransparency=false
      roundingType=HalfHeightRadius
      width=0
      color {colorUseRule=UseColor_Specified, specifiedColor=<rgba 0 0 0 0>}
      cornerRadii=<0 0 0 0>
     }
     tintColor {colorUseRule=UseColor_Specified, specifiedColor=<rgba 1 1 1 1>}
     uVRegion {bIsValid=false, max=<0 0>, min=<0 0>}
    }
    colorAndOpacity=<rgba 1 0 0 1>
    renderTransform {angle=0, scale=<1 1>, shear=<-14 0>, translation=<0 0>}
   }
   slot {class=OverlaySlot, properties {horizontalAlignment=HAlign_Fill, verticalAlignment=VAlign_Fill}}
  }
  HorizontalBox_0 {
   widgetClassPath=&HorizontalBox
   parent=RootOverlay
   widget=HorizontalBox_0
   slot {class=OverlaySlot, properties {verticalAlignment=VAlign_Center, padding=<ltrb 15 0 23 0>}}
  }
  Image_RecDot {
   widgetClassPath=&Image
   parent=HorizontalBox_0
   widget=Image_RecDot
   properties {
    brush {
     bIsDynamicallyLoaded=false
     drawAs=Image
     imageType=NoImage
     mirroring=NoMirror
     resourceName=None
     resourceObject=/Game/UI/Materials/M_UICircle
     tiling=NoTile
     imageSize=<16 16>
     margin=<ltrb 0 0 0 0>
     outlineSettings {
      bUseBrushTransparency=false
      roundingType=HalfHeightRadius
      width=0
      color {colorUseRule=UseColor_Specified, specifiedColor=<rgba 0 0 0 0>}
      cornerRadii=<0 0 0 0>
     }
     tintColor {colorUseRule=UseColor_Specified, specifiedColor=<rgba 1 1 1 1>}
     uVRegion {bIsValid=false, max=<0 0>, min=<0 0>}
    }
    renderTransform {angle=0, scale=<1 1>, shear=<0 0>, translation=<0 -4>}
   }
   slot {class=HorizontalBoxSlot, properties {horizontalAlignment=HAlign_Center, verticalAlignment=VAlign_Center}}
  }
  Text_Callout {
   widgetClassPath=&CommonTextBlock
   bIsVariable=true
   parent=HorizontalBox_0
   widget=Text_Callout
   properties {
    text=A LIVE EVENT IS HAPPENING NOW
    font {
     bForceMonospaced=false
     bMaterialIsStencil=false
     fontMaterial=null
     fontObject=&Burbank
     letterSpacing=0
     monospacedWidth=1
     size=20
     skewAmount=0
     typefaceFontName=Burbank-Black
     outlineSettings {
      bApplyOutlineToDropShadows=false
      bMiteredCorners=false
      bSeparateFillAlpha=false
      outlineMaterial=null
      outlineSize=0
      outlineColor=<rgba 0 0 0 1>
     }
    }
    renderTransform {angle=0, scale=<1 1>, shear=<-10 0>, translation=<0 0>}
   }
   slot {class=HorizontalBoxSlot, properties {padding=<ltrb 10 0 0 2>}}
  }
 ]
}
